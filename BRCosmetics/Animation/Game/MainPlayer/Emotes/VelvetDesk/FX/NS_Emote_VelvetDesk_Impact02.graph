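# Niagara impact effect for the VelvetDesk emote, pairing a delayed hit burst with a short skeletal-mesh socket ribbon trail.
# Ring, speed-line, toon flare, spark and smoke-ribbon emitters burst after a delay chosen by SwingSelect; a separate sprite impact emitter is disabled.
# ColorSelect blends white to blue and varies sizes, with time-of-day tinting on several layers.
# NormalizedNotifyProgress drives smoke turbulence and drift, while RibbonVisibility scales trail opacity.
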NiagaraSystem /BRCosmetics/Animation/Game/MainPlayer/Emotes/VelvetDesk/FX/NS_Emote_VelvetDesk_Impact02
dependencies {
 usedDataInterfaces [
  &NiagaraDataInterfaceArrayFloat
  &NiagaraDataInterfaceCurve
  &NiagaraDataInterfaceArrayFloat3
  &NiagaraDataInterfaceVector2DCurve
  &NiagaraDataInterfaceColorCurve
  &NiagaraDataInterfaceArrayInt32
  &NiagaraDataInterfaceSocketReader
  &NiagaraDataInterfaceSkeletalMesh
 ]
 usedDynamicInputs [&MaskLinearColorBySpawnGroup, &UniformRangedVector2D]
 usedModules [
  &self:SystemSpawnScript.SystemScriptSource.SystemScriptGraph.NiagaraNodeAssignment_0.SetVariables_#1
  &SystemState
  &self:SystemSpawnScript.SystemScriptSource.SystemScriptGraph.NiagaraNodeAssignment_1.SetVariables_#2
 ]
}
emitters {
 Main_RingCombined {
  emitterUpdateScript [
   &EmitterState(
    Life Cycle Mode=::NewEnumerator1
    Loop Behavior=::NewEnumerator1
    Loop Delay=&SelectFloatFromArray(
     Direct Array Index=&MakeIntFromFloat(Float=$User.SwingSelect)
     Float Selection Array=DataInterface() {propertyValues {FloatData [0.12,0.15,0.15]}}
    )
    Loop Duration=1
    UseLoopDelay=true
   )
   &SpawnBurst_Instantaneous(Use Spawn Probability=false, UseLoopCountLimit=false)
  ]
  particleSpawnScript [
   /SetVariables(Particles.RandomRange:NiagaraFloat=&Multiply_Float(
     A=&RandomRangeFloat(Minimum=0.25, Maximum=0.375)
     B=&FloatFromCurve(
      FloatCurve=DataInterface() {propertyValues {Curve {keys [<at 0 1 RCIM_Cubic>, {arriveTangent=7.3942, interpMode=RCIM_Cubic, leaveTangent=7.3942, tangentMode=RCTM_User, time=0.3, value=2}]}}}
      CurveIndex=$Emitter.Age
     )
    )) {isSetParameters=true}
   &InitializeParticle(
    Color=&FNTimeOfDayColor(
     Night Color=<rgba 0.8 0.9367 1 0.6>
     Morning Color=<rgba 0.5958 0.6976 0.7448 0.6>
     Day Color=<rgba 0.8 0.9367 1 0.6>
     Evening Color=<rgba 0.575 0.6732 0.7188 0.6>
    )
    Lifetime=0.6
    Mesh Scale=&Multiply_VectorByFloat(Vector=<1 1.5 1>, Float=&Lerp_Float(A=0.16, B=0.3, Alpha=$User.ColorSelect))
    Mesh Scale Mode=::NewEnumerator4
    UsePositionOffset=false
    Write Material Random=true
   )
   &SystemLocation(MaskSpawnGroup=false, Offset=&Lerp_Vector(A=<0 7.0031 0>, B=<1 9.2921 1>, Alpha=$User.ColorSelect))
   &CameraOffset(Camera Offset Amount=15)
   &InitialMeshOrientation(
    Mesh Axis To Orient=<0 0 0>
    Mesh Orientation Mode=::NewEnumerator2
    Orientation Vector=<0 0 1>
    Rotation=&RandomRangeVector(Minimum=<0 0 0>, Maximum=<0 1 0>)
    Use Rotation Vector=true
   )
  ]
  particleUpdateScript [
   &ParticleState()
   &ScaleMeshSize(Scale Factor=&VectorFromFloat(Value=&FloatFromCurve(
      FloatCurve=DataInterface() {propertyValues {Curve {keys [
          {arriveTangent=5.6347, interpMode=RCIM_Cubic, leaveTangent=5.6347, tangentMode=RCTM_User, time=0, value=0.3788}
          {arriveTangent=1.1451, interpMode=RCIM_Cubic, leaveTangent=1.1451, time=0.2, value=0.85}
          {arriveTangent=0.2133, interpMode=RCIM_Cubic, leaveTangent=0.2133, tangentMode=RCTM_User, time=0.6186, value=1.0872}
          <at 0.9985 1.2296 RCIM_Cubic>
         ]}}}
      Scale Curve=0.95
     )))
   &DynamicMaterialParameters(
    Index 0 Param 1=&FloatFromCurve(FloatCurve=DataInterface() {propertyValues {Curve {keys [
         {arriveTangent=2.3202, interpMode=RCIM_Cubic, leaveTangent=2.3202, tangentMode=RCTM_User, time=0, value=0.2481}
         {arriveTangent=0.6377, interpMode=RCIM_Cubic, leaveTangent=0.6377, time=0.0995, value=0.4178}
         {interpMode=RCIM_Cubic, tangentMode=RCTM_User, time=0.3951, value=0.5}
         {arriveTangent=-1.1781, tangentMode=RCTM_User, time=1, value=0}
        ]}}})
    Module.Index0Param0Write=true
    Module.Index0Param1Write=false
    Module.Index0Param2Write=false
    Module.Index0Param3Write=false
   )
   &ScaleColor(
    Scale Alpha=&FloatFromCurve(
     FloatCurve=DataInterface() {propertyValues {Curve {keys [
         {arriveTangent=-0.0871, interpMode=RCIM_Cubic, leaveTangent=-0.0871, tangentMode=RCTM_User, time=0, value=0.6231}
         {interpMode=RCIM_Cubic, tangentMode=RCTM_User, time=0.5, value=0.7329}
         {tangentMode=RCTM_User, time=1, value=0}
        ]}}}
     Scale Curve=&FloatFromCurve(
      FloatCurve=DataInterface() {propertyValues {Curve {keys [
          {interpMode=RCIM_Cubic, tangentMode=RCTM_User, time=0, value=1}
          {arriveTangent=-0.0131, interpMode=RCIM_Cubic, leaveTangent=0.1095, tangentMode=RCTM_Break, time=0.3, value=1}
          {arriveTangent=-3.0223, interpMode=RCIM_Cubic, leaveTangent=-3.0223, tangentMode=RCTM_User, time=0.7, value=0}
         ]}}}
      CurveIndex=$Emitter.Age
      Scale Curve=0.4
     )
    )
    Scale RGB=&MakeVectorFromLinearColorRGB(Color=&LerpLinearColors(
      StartColor=&ColorFromCurve(ColorCurve=DataInterface() {propertyValues {
         AlphaCurve {keys []}
         BlueCurve {keys [
           <at 0 1 RCIM_Cubic>
           {arriveTangent=-1.2161, interpMode=RCIM_Cubic, leaveTangent=-1.2161, time=0.1313, value=1}
           <at 0.4231 0.4855>
           <at 1 0.2309>
          ]}
         GreenCurve {keys [
           <at 0 1 RCIM_Cubic>
           {arriveTangent=-1.2767, interpMode=RCIM_Cubic, leaveTangent=-1.2767, time=0.1313, value=0.8855}
           <at 0.4231 0.4599>
           <at 1 0.2157>
          ]}
         RedCurve {keys [
           <at 0 1 RCIM_Cubic>
           {arriveTangent=-1.3057, interpMode=RCIM_Cubic, leaveTangent=-1.3057, time=0.1313, value=0.8293}
           <at 0.4231 0.4476>
           <at 1 0.2075>
          ]}
        }})
      EndColor=&ColorFromCurve(ColorCurve=DataInterface() {propertyValues {
         AlphaCurve {keys []}
         BlueCurve {keys [
           <at 0 1 RCIM_Cubic>
           {arriveTangent=-1.2161, interpMode=RCIM_Cubic, leaveTangent=-1.2161, time=0.1313, value=1}
           <at 0.4231 0.4855>
           <at 1 0.2309>
          ]}
         GreenCurve {keys [
           <at 0 1 RCIM_Cubic>
           {arriveTangent=-1.7814, interpMode=RCIM_Cubic, leaveTangent=-1.7814, time=0.1313, value=0.3666}
           <at 0.4231 0.2464>
           <at 1 0.1485>
          ]}
         RedCurve {keys [
           <at 0 1 RCIM_Cubic>
           {arriveTangent=-2.052, interpMode=RCIM_Cubic, leaveTangent=-2.052, time=0.1313, value=0.0552}
           <at 0.4231 0.1319>
           <at 1 0.1043>
          ]}
        }})
      LerpFactor=$User.ColorSelect
     ))
    ScaleA=true
    ScaleRGB=true
   )
   &UpdateMeshOrientation(Rotation Rate=0.2, Rotation Vector=<0 1 0>) {module=UpdateMeshOrientation001}
  ]
  rendererClasses [&NiagaraMeshRendererProperties]
  renderers [{rendererClass=&NiagaraMeshRendererProperties, rendererIndex=0}]
 }
 NE_ImpactCombined {
  bEnabled=false
  emitterUpdateScript [
   &EmitterState(Life Cycle Mode=::NewEnumerator1, Loop Behavior=::NewEnumerator1, Loop Duration=0.4, UseLoopDelay=false)
   &SpawnBurst_Instantaneous(Spawn Count=3, Use Spawn Probability=true, UseLoopCountLimit=false)
  ]
  particleSpawnScript [
   &InitializeParticle(
    Color Minimum=<rgba 0.0503 0.0503 0.0503 1>
    Color Mode=::NewEnumerator2
    Lifetime Max=0.2
    Lifetime Min=0.15
    Lifetime Mode=::NewEnumerator1
    Sprite Size Max=<8 25>
    Sprite Size Min=<10 15>
    Sprite Size Mode=::NewEnumerator2
    Sprite UV Mode=::NewEnumerator1
    UsePositionOffset=false
    Write Material Random=true
   ) {module=InitializeParticle001}
   &AddVelocity(Module.Use Falloff Distance=false, Override Seed=false, Velocity Mode=::NewEnumerator1) {enabled=false, module=AddVelocity001}
   &SubUVAnimation(End Frame=3, SubUV Animation Mode:ENiagara_SubUVLookupMode=::NewEnumerator2)
   &AddVelocity(Velocity=$System.DirectionVector, Velocity Speed Scale=100) {enabled=false}
   &ShapeLocation(
    MaskSpawnGroup=false
    Non Uniform Scale=<1 1 0.5>
    Offset=&MakeVector(X=0, Y=0, Z=-0.3058)
    Offset Mode=::NewEnumerator0
    Sphere Distribution=::NewEnumerator1
    Sphere Radius=1
    U Position=&ReturnNormalizedExecIndex()
    UseOwnerScale=true
   )
   &AddVelocity(
    Cone Angle=43.5386
    Cone Axis=<0 1 0.2>
    Module.UseConeFalloff=false
    Module.UseVelDistribution=false
    Override Seed=false
    Velocity Mode=::NewEnumerator2
    Velocity Speed=&Multiply_Float(A=&RandomRangeFloat(Minimum=0.1, Maximum=1), B=1)
   ) {enabled=false, module=AddVelocity002}
   &AddVelocity(Velocity=&Add_Vector(
     A=&SelectVectorFromArray(
      Direct Array Index=&ReturnExecIndex()
      Vector Selection Array=DataInterface() {propertyValues {FloatData [<0 3 0>, <1 3 0>, <-1 3 0>]}}
     )
     B=&RandomRangeVector(Minimum=<0.1 0.1 0.1>, Maximum=<-0.1 -0.1 -0.1>)
    )) {module=AddVelocity003}
   &CameraOffset(Camera Offset Amount=25)
  ]
  particleUpdateScript [
   &ParticleState()
   &SolveForcesAndVelocity(Clamp Velocity=false, Limit Acceleration=false)
   &ScaleSpriteSize(Scale Factor=&Vector2DFromCurve(
     Vector2Curve=DataInterface() {propertyValues {XCurve {keys [<at 0 1 RCIM_Cubic>,<at 1 1 RCIM_Cubic>]}, YCurve {keys [<at 0 1.6708 RCIM_Cubic>,<at 1 2.8057 RCIM_Cubic>]}}}
     ScaleCurve=&Multiply_Vector2D_ByFloat(Float=1, Vector2D=<1.5 1>)
    ))
   &ScaleColor(
    Scale Alpha=&FloatFromCurve(
     FloatCurve=DataInterface() {propertyValues {Curve {keys [<at 0 1 RCIM_Cubic>,<at 1 0>]}}}
     Scale Curve=&Lerp_Float(
      A=&FNTimeOfDayScalar(Night Scale=0.3, Morning Scale=0.7, Evening Scale=0.7)
      B=1
      Alpha=$User.ColorSelect
     )
    )
    Scale RGB=&MakeVectorFromLinearColorRGB(Color=&LerpLinearColors(
      StartColor=&ColorFromCurve(ColorCurve=DataInterface() {propertyValues {
         AlphaCurve {keys []}
         BlueCurve {keys [<at 0 1 RCIM_Cubic>,<at 0.7087 0.2633 RCIM_Cubic>]}
         GreenCurve {keys [<at 0 1 RCIM_Cubic>,<at 0.7087 0.2633 RCIM_Cubic>]}
         RedCurve {keys [<at 0 1 RCIM_Cubic>,<at 0.7087 0.2633 RCIM_Cubic>]}
        }})
      EndColor=&FNTimeOfDayColor(
       Night Color=&ColorFromCurve(ColorCurve=DataInterface() {propertyValues {
          AlphaCurve {keys []}
          BlueCurve {keys [<at 0 1 RCIM_Cubic>,<at 0.7087 0.3189 RCIM_Cubic>]}
          GreenCurve {keys [<at 0 0.5954 RCIM_Cubic>,<at 0.7087 0.0791 RCIM_Cubic>]}
          RedCurve {keys [<at 0 0 RCIM_Cubic>,<at 0.7087 0 RCIM_Cubic>]}
         }})
       Morning Color=&ColorFromCurve(ColorCurve=DataInterface() {propertyValues {
          AlphaCurve {keys []}
          BlueCurve {keys [<at 0 1 RCIM_Cubic>,<at 0.7087 0.3189 RCIM_Cubic>]}
          GreenCurve {keys [<at 0 0.6638 RCIM_Cubic>,<at 0.7087 0.0791 RCIM_Cubic>]}
          RedCurve {keys [<at 0 0.1691 RCIM_Cubic>,<at 0.7087 0 RCIM_Cubic>]}
         }})
       Day Color=&ColorFromCurve(ColorCurve=DataInterface() {propertyValues {
          AlphaCurve {keys []}
          BlueCurve {keys [<at 0 1 RCIM_Cubic>,<at 0.7087 0.3189 RCIM_Cubic>]}
          GreenCurve {keys [<at 0 0.6638 RCIM_Cubic>,<at 0.7087 0.0791 RCIM_Cubic>]}
          RedCurve {keys [<at 0 0.1691 RCIM_Cubic>,<at 0.7087 0 RCIM_Cubic>]}
         }})
       Evening Color=&ColorFromCurve(ColorCurve=DataInterface() {propertyValues {
          AlphaCurve {keys []}
          BlueCurve {keys [<at 0 1 RCIM_Cubic>,<at 0.7087 0.3189 RCIM_Cubic>]}
          GreenCurve {keys [<at 0 0.6638 RCIM_Cubic>,<at 0.7087 0.0791 RCIM_Cubic>]}
          RedCurve {keys [<at 0 0.1691 RCIM_Cubic>,<at 0.7087 0 RCIM_Cubic>]}
         }})
      )
      LerpFactor=$User.ColorSelect
     ))
    ScaleA=true
    ScaleRGB=true
   )
  ]
  rendererClasses [&NiagaraSpriteRendererProperties]
  renderers [{rendererClass=&NiagaraSpriteRendererProperties}]
 }
 NE_SmokeRibbonCombined {
  emitterUpdateScript [
   &EmitterState(
    Life Cycle Mode=::NewEnumerator1
    Loop Behavior=::NewEnumerator1
    Loop Delay=&SelectFloatFromArray(
     Direct Array Index=&MakeIntFromFloat(Float=$User.SwingSelect)
     Float Selection Array=DataInterface() {propertyValues {FloatData [0.12,0.15,0.15]}}
    )
    Loop Duration=1
    UseLoopDelay=true
   )
   &SpawnBurst_Instantaneous(Spawn Count=20, Use Spawn Probability=false, UseLoopCountLimit=false)
   /SetVariables(Emitter.RandomSeed=1) {enabled=false, isSetParameters=true}
  ]
  particleSpawnScript [
   /SetVariables(Particles.RibbonLinkOrder:NiagaraFloat=&MakeFloatFromInt(Int=&ReturnExecIndex())) {isSetParameters=true}
   &InitializeParticle(
    Color=<rgba 1 1 1 0.05>
    Lifetime=&Lerp_Float(A=0.4, B=0.8, Alpha=$User.ColorSelect)
    Position Offset=&Add_Vector(
     A=<0 0 0>
     B=&MakeVector(X=0, Y=&ReturnNormalizedExecIndex(Normalized Index Scale=200), Z=0)
    )
    Ribbon Width=10
    Ribbon Width Mode=::NewEnumerator1
    UsePositionOffset=false
    Write Material Random=true
   )
  ]
  particleUpdateScript [
   &ParticleState()
   /SetVariables(Particles.Position=&AddVectorToPosition(
     Position=&MakeVector(X=0, Y=&ReturnNormalizedExecIndex(), Z=0)
     Vector=&MakeVector(
      X=0
      Y=&ReturnNormalizedExecIndex(Normalized Index Scale=&FloatFromCurve(
        FloatCurve=DataInterface() {propertyValues {Curve {keys [
            {arriveTangent=3.8864, interpMode=RCIM_Cubic, leaveTangent=3.8864, leaveTangentWeight=1.1238, tangentMode=RCTM_User, time=0, value=0}
            {arriveTangent=1.3836, arriveTangentWeight=0.1796, interpMode=RCIM_Cubic, leaveTangent=1.3836, tangentMode=RCTM_User, time=0.3515, value=1}
           ]}}}
        Scale Curve=260
       ))
      Z=0
     )
    )) {isSetParameters=true}
   &CurlNoiseForce(
    Noise Frequency=200
    Noise Strength=&Add_Float(
     A=&FloatFromCurve(
      FloatCurve=DataInterface() {propertyValues {Curve {keys [{interpMode=RCIM_Cubic, tangentMode=RCTM_User, time=0, value=0.5}, {interpMode=RCIM_Cubic, tangentMode=RCTM_User, time=0.5, value=1}]}}}
      CurveIndex=$User.NormalizedNotifyProgress
      Scale Curve=800
     )
     B=&FloatFromCurve(FloatCurve=DataInterface() {propertyValues {Curve {keys [<at 0 1>,<at 1 0>]}}}, CurveIndex=$Particles.RibbonLinkOrder, Scale Curve=400)
    )
    Pan Noise=true
    Pan Noise Field=<0 0 1>
   ) {module=CurlNoiseForce001}
   /SetVariables(Transient.PhysicsForce:Vector3f=&Add_Vector(
     A=&Multiply_Vector(A=$Output.CurlNoiseForce001.CurlNoiseForce, B=<0.5 0 0.5>)
     B=&Add_Vector(
      A=&Multiply_VectorByFloat(
       Vector=<-1 0 0>
       Float=&FloatFromCurve(
        FloatCurve=DataInterface() {propertyValues {Curve {keys [
            {arriveTangent=5.1592, interpMode=RCIM_Cubic, leaveTangent=5.1592, tangentMode=RCTM_User, time=0.45, value=0}
            {interpMode=RCIM_Cubic, tangentMode=RCTM_User, time=1, value=1}
           ]}}}
        CurveIndex=$User.NormalizedNotifyProgress
        Scale Curve=800
       )
      )
      B=&Multiply_VectorByFloat(
       Vector=<0 20 0>
       Float=&FloatFromCurve(
        FloatCurve=DataInterface() {propertyValues {Curve {keys [
            {arriveTangent=2.7159, interpMode=RCIM_Cubic, leaveTangent=2.7159, tangentMode=RCTM_User, time=0, value=0}
            {interpMode=RCIM_Cubic, tangentMode=RCTM_User, time=1, value=1}
           ]}}}
        Scale Curve=400
       )
      )
     )
    )) {isSetParameters=true}
   &ScaleRibbonWidth(Ribbon Width Scale=&Multiply_Float(
     A=&FloatFromCurve(
      FloatCurve=DataInterface() {propertyValues {Curve {keys [
          {arriveTangent=0.9253, interpMode=RCIM_Cubic, leaveTangent=0.9253, tangentMode=RCTM_User, time=0.1, value=1.1823}
          {arriveTangent=1.0217, interpMode=RCIM_Cubic, leaveTangent=1.0217, tangentMode=RCTM_User, time=0.2185, value=1.3188}
          {arriveTangent=-2.5115, interpMode=RCIM_Cubic, leaveTangent=-2.5115, tangentMode=RCTM_User, time=1, value=0.7837}
         ]}}}
      Scale Curve=1.4
     )
     B=&FloatFromCurve(
      FloatCurve=DataInterface() {propertyValues {Curve {keys [
          {interpMode=RCIM_Cubic, tangentMode=RCTM_User, time=0, value=0.563}
          {interpMode=RCIM_Cubic, tangentMode=RCTM_User, time=0.0836, value=1.0182}
          {interpMode=RCIM_Cubic, tangentMode=RCTM_User, time=0.3, value=1.5}
          {arriveTangent=-1.6746, interpMode=RCIM_Cubic, leaveTangent=-1.6746, tangentMode=RCTM_User, time=1, value=0.8282}
         ]}}}
      CurveIndex=&ReturnNormalizedExecIndex()
     )
    )) {module=ScaleRibbonWidth001}
   &Drag(Drag=2, Use Linear Drag=true, Use Rotational Drag=true)
   &SolveForcesAndVelocity(Clamp Velocity=false, Limit Acceleration=false)
   &ScaleColor(
    Scale Alpha=&Lerp_Float(
     A=&FloatFromCurve(FloatCurve=DataInterface() {propertyValues {Curve {keys [
          {interpMode=RCIM_Cubic, tangentMode=RCTM_User, time=0.0439, value=1}
          {arriveTangent=-3.6022, interpMode=RCIM_Cubic, leaveTangent=-3.6022, tangentMode=RCTM_User, time=0.877, value=0}
         ]}}})
     B=&FloatFromCurve(FloatCurve=DataInterface() {propertyValues {Curve {keys [
          {arriveTangent=-168.7339, interpMode=RCIM_Cubic, leaveTangent=-168.7339, tangentMode=RCTM_User, time=0.0439, value=9.3789}
          {arriveTangent=-9.8094, interpMode=RCIM_Cubic, leaveTangent=-9.8094, time=0.1361, value=2.6843}
          {interpMode=RCIM_Cubic, tangentMode=RCTM_User, time=1, value=0}
         ]}}})
     Alpha=$User.ColorSelect
    )
    Scale RGB=&Lerp_Vector(
     A=<1 1 1>
     B=&MakeVectorFromLinearColorRGB(Color=&ColorFromCurve(ColorCurve=DataInterface() {propertyValues {
         AlphaCurve {keys []}
         BlueCurve {keys [<at 0.0727 1 RCIM_Cubic>,<at 0.444 1 RCIM_Cubic>]}
         GreenCurve {keys [<at 0.0727 0.4973 RCIM_Cubic>,<at 0.444 1 RCIM_Cubic>]}
         RedCurve {keys [<at 0.0727 0 RCIM_Cubic>,<at 0.444 1 RCIM_Cubic>]}
        }}))
     Alpha=$User.ColorSelect
    )
    ScaleA=true
    ScaleRGB=true
   ) {module=ScaleColor001}
  ]
  rendererClasses [&NiagaraRibbonRendererProperties]
  renderers [{rendererClass=&NiagaraRibbonRendererProperties}]
 }
 NE_SparksCombined {
  emitterUpdateScript [
   &EmitterState(
    Life Cycle Mode=::NewEnumerator1
    Loop Behavior=::NewEnumerator1
    Loop Delay=&SelectFloatFromArray(
     Direct Array Index=&MakeIntFromFloat(Float=$User.SwingSelect)
     Float Selection Array=DataInterface() {propertyValues {FloatData [0.12,0.15,0.15]}}
    )
    Loop Duration=1
    UseLoopDelay=true
   )
   &SpawnBurst_Instantaneous(Spawn Count=&RandomRangeInt(Minimum=7, Maximum=5), Use Spawn Probability=false)
  ]
  particleSpawnScript [
   /SetVariables(Particles.Random=&RandomRangeFloat(Minimum=0, Maximum=1)) {isSetParameters=true}
   &InitializeParticle(
    Lifetime=&Lerp_Float(A=0.3, B=0.2, Alpha=$Particles.Random)
    Sprite Size Max=&Lerp_Vector2(A=<3 15>, B=<10 65>, Alpha=$User.ColorSelect)
    Sprite Size Min=&Lerp_Vector2(A=<5 30>, B=<5 40>, Alpha=$User.ColorSelect)
    Sprite Size Mode=::NewEnumerator2
    UsePositionOffset=false
    Write Material Random=true
   ) {module=InitializeParticle001}
   &CameraOffset(Camera Offset Amount=30)
   &SphereLocation(
    MaskSpawnGroup=false
    Override Seed=false
    Sphere Radius=1
    Surface Only=false
    UseNonUniformScale=false
    UseOffset=false
    UseSphereOrientation=false
   )
   &AddVelocity(
    Scale Added Velocity=&Multiply_VectorByFloat(Vector=<1 1 1>, Float=&Lerp_Float(A=2, B=3, Alpha=$Particles.Random))
    Velocity=&RandomRangeVector(Minimum=<-50 30 -30>, Maximum=<50 50 50>)
   ) {module=AddVelocity001}
   &DynamicMaterialParameters(
    Index 0 Param 2=&FloatFromCurve(
     FloatCurve=DataInterface() {propertyValues {Curve {keys [<at 0 0 RCIM_Cubic>, {arriveTangent=1, interpMode=RCIM_Cubic, leaveTangent=1, time=0.7035, value=0.3698}, <at 1 1>]}}}
     CurveIndex=&RandomRangeFloat(Minimum=0, Maximum=1)
    )
    Module.Param0WriteEnabled=true
    Module.Param1WriteEnabled=true
    Module.Param2WriteEnabled=true
    Module.Param3WriteEnabled=true
   ) {module=DynamicMaterialParameters001}
  ]
  particleUpdateScript [
   &ParticleState()
   &SolveForcesAndVelocity(Clamp Velocity=false, Limit Acceleration=false)
   &DynamicMaterialParameters(
    Index 0 Param 1=&FloatFromCurve(FloatCurve=DataInterface() {propertyValues {Curve {keys [<at 0 0>,<at 1 1>]}}})
    Module.Param0WriteEnabled=true
    Module.Param1WriteEnabled=false
    Module.Param2WriteEnabled=true
    Module.Param3WriteEnabled=true
   )
   &ScaleColor(
    Scale Alpha=&FloatFromCurve(FloatCurve=DataInterface() {propertyValues {Curve {keys [<at 0 1>,<at 1 0>]}}})
    Scale RGB=&MakeVectorFromLinearColorRGB(Color=&FNTimeOfDayColor(
      Night Color=&ColorFromCurve(ColorCurve=DataInterface() {propertyValues {AlphaCurve {keys [<at 0 1>,<at 1 0>]}, BlueCurve {keys [<at 0 0.3883>]}, GreenCurve {keys [<at 0 0.2944>]}, RedCurve {keys [<at 0 0.2106>]}}})
      Morning Color=&ColorFromCurve(ColorCurve=DataInterface() {propertyValues {AlphaCurve {keys [<at 0 1>,<at 1 0>]}, BlueCurve {keys [<at 0 0.5535>]}, GreenCurve {keys [<at 0 0.5482>]}, RedCurve {keys [<at 0 0.7679>]}}})
      Day Color=&ColorFromCurve(ColorCurve=DataInterface() {propertyValues {AlphaCurve {keys [<at 0 1>,<at 1 0>]}, BlueCurve {keys [<at 0 0.7679>]}, GreenCurve {keys [<at 0 0.7679>]}, RedCurve {keys [<at 0 0.7679>]}}})
      Evening Color=&ColorFromCurve(ColorCurve=DataInterface() {propertyValues {AlphaCurve {keys [<at 0 1>,<at 1 0>]}, BlueCurve {keys [<at 0 0.5897>]}, GreenCurve {keys [<at 0 0.5314>]}, RedCurve {keys [<at 0 0.4795>]}}})
     ))
    ScaleA=true
    ScaleRGB=true
   ) {module=ScaleColor001}
  ]
  rendererClasses [&NiagaraSpriteRendererProperties]
  renderers [{rendererClass=&NiagaraSpriteRendererProperties}]
 }
 NE_SpeedLinesCombined {
  emitterUpdateScript [
   &EmitterState(
    Life Cycle Mode=::NewEnumerator1
    Loop Behavior=::NewEnumerator1
    Loop Delay=&SelectFloatFromArray(
     Direct Array Index=&MakeIntFromFloat(Float=$User.SwingSelect)
     Float Selection Array=DataInterface() {propertyValues {FloatData [0.12,0.15,0.15]}}
    )
    Loop Duration=1
    UseLoopDelay=true
   )
   &SpawnBurst_Instantaneous(Spawn Count=2, Use Spawn Probability=false)
   /SetVariables(
    Emitter.Bright:LinearColor=&LerpLinearColors(StartColor=<rgba 5 5 5 1>, EndColor=<rgba 0 1.9427 5 1>, LerpFactor=$User.ColorSelect)
    Emitter.Bright_Y:LinearColor=&LerpLinearColors(StartColor=<rgba 6.7574 6.3388 4.3518 1>, EndColor=<rgba 1.1239 5.0417 6.7574 1>, LerpFactor=$User.ColorSelect)
    Emitter.Dark:LinearColor=&LerpLinearColors(StartColor=<rgba 0.2917 0.2917 0.2917 1>, EndColor=<rgba 0.2917 0.2917 0.2917 1>, LerpFactor=$User.ColorSelect)
    Emitter.Dark_Y:LinearColor=&LerpLinearColors(StartColor=<rgba 2 2 2 1>, EndColor=<rgba 0 1.0409 2 1>, LerpFactor=$User.ColorSelect)
   ) {isSetParameters=true}
  ]
  particleSpawnScript [
   /Niagara/Modules/Spawn/Initialization/InitializeParticle(
    Color=<rgba 1 1 1 5>
    Lifetime=0.24
    Mesh Scale=&VectorFromFloat(Value=&Multiply_Float(A=2.5, B=&RandomRangeFloat(Minimum=1, Maximum=1.8)))
    Module.Write Color=true
    Module.Write Lifetime=true
    Module.Write Mass=true
    Module.Write Material Random=true
    Module.Write Position=false
    Module.Write Scale=true
    Module.Write SpriteRotation=false
    Module.Write SpriteSize=true
    Position=&SimulationPosition()
    Sprite Size=<5 5>
   ) {enabled=false}
   /Niagara/Modules/Spawn/Initialization/V2/InitializeParticle(
    Color=<rgba 1 1 1 5>
    Lifetime=0.24
    Mesh Scale=&Multiply_VectorByFloat(Vector=<1 1 1.0405>, Float=&Lerp_Float(A=0.6012, B=0.6723, Alpha=$User.ColorSelect))
    Mesh Scale Mode=::NewEnumerator4
    UsePositionOffset=false
    Write Material Random=true
   ) {module=InitializeParticle001}
   &PartitionParticles()
   &InitialMeshOrientation(
    Mesh Axis To Orient=<0 0 1>
    Mesh Orientation Mode=::NewEnumerator2
    Orientation Vector=<0 0 1>
    Rotation=<0.25 0 0>
    Use Rotation Vector=true
   )
   &SystemLocation(MaskSpawnGroup=false) {enabled=false}
  ]
  particleUpdateScript [
   &ParticleState()
   &SolveForcesAndVelocity(Clamp Velocity=false, Limit Acceleration=false)
   &DynamicMaterialParameters(
    Index 0 Param 1=1
    Index 0 Param 2=&RandomRangeFloat(Minimum=1.5, Maximum=2)
    Index 0 Param 3=0.2
    Index 0 Param 4=&SelectFloatFromArray(Direct Array Index=$Particles.PartitionParticles.Partition, Float Selection Array=DataInterface() {propertyValues {FloatData [0,0.1]}})
    Module.Param0WriteEnabled=true
    Module.Param1WriteEnabled=true
    Module.Param2WriteEnabled=true
    Module.Param3WriteEnabled=true
   )
   &ScaleColor(
    Scale Alpha=&FloatFromCurve(
     FloatCurve=DataInterface() {propertyValues {Curve {keys [
         {arriveTangent=44.6077, interpMode=RCIM_Cubic, leaveTangent=44.6077, tangentMode=RCTM_User, time=0, value=0}
         <at 0.0781 1.0138 RCIM_Cubic>
         {arriveTangent=-1.8607, interpMode=RCIM_Cubic, leaveTangent=-1.8607, tangentMode=RCTM_User, time=1, value=0}
        ]}}}
     Scale Curve=&FNTimeOfDayScalar(Night Scale=0.7, Morning Scale=1, Evening Scale=0.7)
    )
    ScaleA=true
    ScaleRGB=true
   )
   &ScaleMeshSize(Scale Factor=&VectorFromFloat(Value=&FloatFromCurve(
      FloatCurve=DataInterface() {propertyValues {Curve {keys [{arriveTangent=2.9993, interpMode=RCIM_Cubic, leaveTangent=2.9993, tangentMode=RCTM_User, time=0, value=0}, <at 1 1 RCIM_Cubic>]}}}
      Scale Curve=0.8
     )))
   &CameraOffset(Camera Offset Amount=20)
  ]
  rendererClasses [&NiagaraMeshRendererProperties]
  renderers [{rendererClass=&NiagaraMeshRendererProperties, rendererIndex=0}]
 }
 Toon_Impact_FlareCombined {
  emitterUpdateScript [
   &EmitterState(
    Life Cycle Mode=::NewEnumerator1
    Loop Behavior=::NewEnumerator1
    Loop Delay=&SelectFloatFromArray(
     Direct Array Index=&MakeIntFromFloat(Float=$User.SwingSelect)
     Float Selection Array=DataInterface() {propertyValues {FloatData [0.12,0.15,0.15]}}
    )
    Loop Duration=0.5
    UseLoopDelay=true
   )
   &SpawnBurst_Instantaneous(Spawn Count=1, Use Spawn Probability=false)
   &SpawnBurst_Instantaneous(Spawn Count=3, Spawn Group=1, Use Spawn Probability=true, UseLoopCountLimit=false) {module=SpawnBurst_Instantaneous001}
  ]
  particleSpawnScript [
   /SetVariables(Particles.VisibilityTag=&SpawnGroup()) {isSetParameters=true}
   &InitializeParticle(
    Lifetime=&RandomRangeFloat(Minimum=0.15, Maximum=0.2)
    Mesh Scale=&Multiply_VectorByFloat(
     Vector=&RandomRangeVector(Minimum=<1.45 1.45 0.7909>, Maximum=<1.3 1.3 1.1403>)
     Float=&Lerp_Float(A=0.6924, B=0.8469, Alpha=$User.ColorSelect)
    )
    Mesh Scale Mode=::NewEnumerator4
    Position Offset=<0 4.0706 0>
    Sprite Size=<200 175>
    Sprite Size Mode=::NewEnumerator4
    UsePositionOffset=false
    Write Material Random=true
   ) {module=InitializeParticle001}
   &SubUVAnimation(
    End Frame:NiagaraInt32=&SelectIntFromArray(Direct Array Index=&SpawnGroup(), Int Selection Array=DataInterface() {propertyValues {IntData [0,3]}})
    SubUV Animation Mode:ENiagara_SubUVLookupMode=::NewEnumerator2
   )
   &ShapeLocation(
    MaskSpawnGroup=false
    Non Uniform Scale=<1 1 0.5>
    Offset=&MakeVector(X=0, Y=0, Z=-0.3058)
    Offset Mode=::NewEnumerator0
    Sphere Distribution=::NewEnumerator1
    Sphere Radius=1
    U Position=&ReturnNormalizedExecIndex()
    UseOwnerScale=true
   )
   &AddVelocity(Velocity=&Lerp_Vector(
     A=<0 0 0>
     B=&Add_Vector(
      A=&SelectVectorFromArray(
       Direct Array Index=&ReturnExecIndex()
       Vector Selection Array=DataInterface() {propertyValues {FloatData [<0 3 0>, <1 3 0>, <-1 3 0>]}}
      )
      B=&RandomRangeVector(Minimum=<0.1 0.1 0.1>, Maximum=<-0.1 -0.1 -0.1>)
     )
     Alpha=&MakeFloatFromInt(Int=&SpawnGroup())
    )) {module=AddVelocity003}
   /SetVariables(Particles.CoronaSpriteSize=&Vector2DFromFloat(Value=90)) {isSetParameters=true}
   /SetVariables(Particles.ImpactColor:LinearColor=&MakeLinearColorFromVectorAndFloat("Vector (RGB)"=<1 1 1>, "Float (Alpha)"=&Multiply_Float(A=1, B=1))) {isSetParameters=true}
   /SetVariables(Particles.NewImpactSize:Vector2f=&Lerp_Vector2(
     A=&Lerp_Vector2(A=<10 15>, B=<8 25>, Alpha=0.5)
     B=&RandomRangeVector2D(Minimum=<10 20>, Maximum=<15 40>)
     Alpha=$User.ColorSelect
    )) {isSetParameters=true}
   &InitialMeshOrientation(
    Orientation Axis=<0 0 1>
    Rotation=&RandomRangeVector(Minimum=<0 0 0.25>, Maximum=<0 0 0.25>)
    Use Orientation Vector=true
    Use Rotation Vector=true
   )
   &CameraOffset(Camera Offset Amount=&Lerp_Float(A=30, B=25, Alpha=&MakeFloatFromInt(Int=&SpawnGroup()))) {module=CameraOffset001}
  ]
  particleUpdateScript [
   &CameraOffset(Camera Offset Amount=30)
   &ParticleState()
   &DynamicMaterialParameters(
    Index 0 Param 1=&FloatFromCurve(FloatCurve=DataInterface() {propertyValues {Curve {keys [<at 0 1 RCIM_Cubic>,<at 0.8 0 RCIM_Cubic>]}}})
    Index 0 Param 2=2
    Index 0 Param 3=1
    Index 0 Param 4=1
    Index 1 Param 1=&FloatFromCurve(FloatCurve=DataInterface() {propertyValues {Curve {keys [
         {arriveTangent=-2.3163, interpMode=RCIM_Cubic, time=0, value=2}
         {arriveTangent=-1.9019, interpMode=RCIM_Cubic, leaveTangent=-1.9019, tangentMode=RCTM_User, time=1, value=0.4}
        ]}}})
    Index 1 Param 2=0.4
    Index 1 Param 3=0.6
    Index 1 Param 4=1
    Index 2 Param 1=&FloatFromCurve(FloatCurve=DataInterface() {propertyValues {Curve {keys [<at 0.46 0>,<at 1 1>]}}})
    Index 2 Param 2=45
    Module.Param0WriteEnabled=true
    Module.Param0WriteEnabled1=true
    Module.Param0WriteEnabled2=true
    Module.Param1WriteEnabled=true
    Module.Param1WriteEnabled1=true
    Module.Param1WriteEnabled2=true
    Module.Param2WriteEnabled=true
    Module.Param2WriteEnabled1=true
    Module.Param2WriteEnabled2=true
    Module.Param3WriteEnabled=true
    Module.Param3WriteEnabled1=true
    Module.Param3WriteEnabled2=true
    Write Parameter Index 1=true
    Write Parameter Index 2=true
   )
   /SetVariables(Particles.CoronaSpriteColor:LinearColor=&MakeLinearColorFromVectorAndFloat(
     "Vector (RGB)"=&MakeVectorFromLinearColorRGB(Color=<rgba 0.5521 0.5521 0.5521 1>)
     "Float (Alpha)"=&Multiply_Float(
      A=&FloatFromCurve(FloatCurve=DataInterface() {propertyValues {Curve {keys [<at 0.35 1.5 RCIM_Cubic>,<at 0.95 0>]}}})
      B=3.9
     )
    )) {isSetParameters=true}
   /SetVariables(Particles.ImpactColor:LinearColor=&FNTimeOfDayColor(
     Night Color=<rgba 0.4809 0.4809 0.4809 1>
     Morning Color=<rgba 1 1 1 1>
     Day Color=<rgba 1 1 1 1>
     Evening Color=<rgba 0.4809 0.4809 0.4809 1>
    )) {isSetParameters=true}
   /SetVariables(Particles.NewImpactSize:Vector2f=&Multiply_Vector2D(
     A=$Particles.Initial.NewImpactSize
     B=&Vector2DFromCurve(
      Vector2Curve=DataInterface() {propertyValues {XCurve {keys [<at 0 1 RCIM_Cubic>,<at 1 1 RCIM_Cubic>]}, YCurve {keys [<at 0 1.6708 RCIM_Cubic>,<at 1 2.8057 RCIM_Cubic>]}}}
      ScaleCurve=&Multiply_Vector2D_ByFloat(Float=1, Vector2D=<1.5 1>)
     )
    )) {isSetParameters=true}
   &ScaleColor(
    Scale Alpha=&FloatFromCurve(FloatCurve=DataInterface() {propertyValues {Curve {keys [<at 0 1>,<at 1 0>]}}})
    Scale RGB=&Lerp_Vector(
     A=<1 1 1>
     B=&MakeVectorFromLinearColorRGB(Color=&LerpLinearColors(
       StartColor=&ColorFromCurve(ColorCurve=DataInterface() {propertyValues {
          AlphaCurve {keys []}
          BlueCurve {keys [<at 0 1 RCIM_Cubic>,<at 0.7087 0.2633 RCIM_Cubic>]}
          GreenCurve {keys [<at 0 1 RCIM_Cubic>,<at 0.7087 0.2633 RCIM_Cubic>]}
          RedCurve {keys [<at 0 1 RCIM_Cubic>,<at 0.7087 0.2633 RCIM_Cubic>]}
         }})
       EndColor=&FNTimeOfDayColor(
        Night Color=&ColorFromCurve(ColorCurve=DataInterface() {propertyValues {
           AlphaCurve {keys []}
           BlueCurve {keys [<at 0 1 RCIM_Cubic>,<at 0.7087 0.3189 RCIM_Cubic>]}
           GreenCurve {keys [<at 0 0.5954 RCIM_Cubic>,<at 0.7087 0.0791 RCIM_Cubic>]}
           RedCurve {keys [<at 0 0 RCIM_Cubic>,<at 0.7087 0 RCIM_Cubic>]}
          }})
        Morning Color=&ColorFromCurve(ColorCurve=DataInterface() {propertyValues {
           AlphaCurve {keys []}
           BlueCurve {keys [<at 0 1 RCIM_Cubic>,<at 0.7087 0.3189 RCIM_Cubic>]}
           GreenCurve {keys [<at 0 0.6638 RCIM_Cubic>,<at 0.7087 0.0791 RCIM_Cubic>]}
           RedCurve {keys [<at 0 0.1691 RCIM_Cubic>,<at 0.7087 0 RCIM_Cubic>]}
          }})
        Day Color=&ColorFromCurve(ColorCurve=DataInterface() {propertyValues {
           AlphaCurve {keys []}
           BlueCurve {keys [<at 0 1 RCIM_Cubic>,<at 0.7087 0.3189 RCIM_Cubic>]}
           GreenCurve {keys [<at 0 0.6638 RCIM_Cubic>,<at 0.7087 0.0791 RCIM_Cubic>]}
           RedCurve {keys [<at 0 0.1691 RCIM_Cubic>,<at 0.7087 0 RCIM_Cubic>]}
          }})
        Evening Color=&ColorFromCurve(ColorCurve=DataInterface() {propertyValues {
           AlphaCurve {keys []}
           BlueCurve {keys [<at 0 1 RCIM_Cubic>,<at 0.7087 0.3189 RCIM_Cubic>]}
           GreenCurve {keys [<at 0 0.6638 RCIM_Cubic>,<at 0.7087 0.0791 RCIM_Cubic>]}
           RedCurve {keys [<at 0 0.1691 RCIM_Cubic>,<at 0.7087 0 RCIM_Cubic>]}
          }})
       )
       LerpFactor=$User.ColorSelect
      ))
     Alpha=&MakeFloatFromInt(Int=&SpawnGroup())
    )
    ScaleA=true
    ScaleRGB=true
   )
   &ScaleMeshSize(Scale Factor=&VectorFromFloat(Value=&FloatFromCurve(FloatCurve=DataInterface() {propertyValues {Curve {keys [<at 0 0.8071 RCIM_Cubic>,<at 1 1.202>]}}})))
   &SolveForcesAndVelocity(Clamp Velocity=false, Limit Acceleration=false)
  ]
  rendererClasses [&NiagaraSpriteRendererProperties, &NiagaraMeshRendererProperties]
  renderers [
   {rendererClass=&NiagaraSpriteRendererProperties}
   {rendererClass=&NiagaraMeshRendererProperties, rendererIndex=1}
   {rendererClass=&NiagaraSpriteRendererProperties, rendererIndex=2}
  ]
 }
 Trail001 {
  emitterUpdateScript [
   &EmitterState(Inactive Response=::NewEnumerator1, Life Cycle Mode=::NewEnumerator1, Loop Behavior=::NewEnumerator1, Loop Duration=0.4, UseLoopDelay=false)
   /SetVariables(Emitter.NewSpawnRate=60) {isSetParameters=true}
   &SpawnRate(SpawnRate=$Emitter.NewSpawnRate, Use Spawn Probability=false)
  ]
  particleSpawnScript [
   &InitializeParticle(
    Color=&FNTimeOfDayColor(
     Night Color=<rgba 0 0.0615 0.2957 1>
     Morning Color=<rgba 0.0072 0.1607 0.3189 1>
     Day Color=<rgba 0.0578 0.155 0.3189 1>
     Evening Color=<rgba 0.0209 0.1318 0.3189 1>
    )
    Lifetime=0.1
    Ribbon Facing Mode=::NewEnumerator1
    Ribbon Width=100
    Ribbon Width Mode=::NewEnumerator1
    UsePositionOffset=false
    Write Material Random=true
   )
   &SkeletalMeshLocation(MaskSpawnGroup=false, Mesh Sampling Type=::NewEnumerator5, Skeletal Mesh=$System.Skeletal Mesh, Socket Sampling Mode=::NewEnumerator5)
   /SetVariables(Particles.Orientation:Vector3f=&RotateVectorByQuaternion(Vector To Rotate=<0 1 0>, Quaternion=$Particles.SkeletalMeshLocation.SampledRotation)) {isSetParameters=true}
   &AddVelocity(Velocity=<400 0 0>) {enabled=false}
   /SetVariables() {isSetParameters=true}
  ]
  particleUpdateScript [
   &SkeletalMeshLocation(
    MaskSpawnGroup=false
    Mesh Sampling Type=::NewEnumerator5
    Position Sampling=::NewEnumerator2
    Sampled Position Offset=<0 -22 0>
    Skeletal Mesh=$System.Skeletal Mesh
    Socket Sampling Mode=::NewEnumerator5
   ) {module=SkeletalMeshLocation001}
   &self:Leading_Edge_Fix(Position:NiagaraPosition=$Particles.SkeletalMeshLocation001.SampledPosition) {module=Leading_Edge_Fix}
   &ParticleState()
   &ScaleColor(
    Scale Alpha=&FloatFromCurve(
     FloatCurve=DataInterface() {propertyValues {Curve {keys [<at 0 0 RCIM_Cubic>,<at 0.1 1 RCIM_Cubic>,<at 1 0 RCIM_Cubic>]}}}
     Scale Curve=&Multiply_Float(A=1, B=$User.RibbonVisibility)
    )
    ScaleA=true
    ScaleRGB=true
   )
   &SolveForcesAndVelocity(Clamp Velocity=false, Limit Acceleration=false)
  ]
  rendererClasses [&NiagaraRibbonRendererProperties]
  renderers [{rendererClass=&NiagaraRibbonRendererProperties}]
 }
}
summary {
 systemName=NS_Emote_VelvetDesk_Impact02
 emitters [
  {emitterName=NE_SpeedLinesCombined, rendererClasses [&NiagaraMeshRendererProperties]}
  {bEnabled=false, emitterName=NE_ImpactCombined, rendererClasses [&NiagaraSpriteRendererProperties]}
  {emitterName=NE_SmokeRibbonCombined, rendererClasses [&NiagaraRibbonRendererProperties]}
  {emitterName=Main_RingCombined, rendererClasses [&NiagaraMeshRendererProperties]}
  {
   emitterName=Toon_Impact_FlareCombined
   rendererClasses [&NiagaraSpriteRendererProperties, &NiagaraMeshRendererProperties]
  }
  {emitterName=NE_SparksCombined, rendererClasses [&NiagaraSpriteRendererProperties]}
  {emitterName=Trail001, rendererClasses [&NiagaraRibbonRendererProperties]}
 ]
 userVariables [
  User.ColorSelect {type=NiagaraFloat, defaultValue=1, description=""}
  User.NormalizedNotifyProgress {type=NiagaraFloat, defaultValue=0.4, description=""}
  User.RibbonVisibility {type=NiagaraFloat, defaultValue=1, description=""}
  User.SmokeHeight {type=NiagaraFloat, defaultValue=1, description=""}
  User.SwingSelect {type=NiagaraFloat, defaultValue=0, description=""}
 ]
}
